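# Cube_AmbPan_Cue is a Sound Cue for an ambient rune-pan sound of the Cube, classed under scripted event audio.
# A Quality Level root picks between the Cube_Positive_RunePan_01 wave and its "_opt" variant, each covering two of four quality slots.
# Playback uses the shared Cube ambient zaps attenuation and its own concurrency override, which caps instances, stops the oldest and scales volume down as copies stack.
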SoundCue /Game/Athena/Sounds/Events/Cube/Cube_AmbPan_Cue
root=QualityLevel_0
nodes [
 {id=QualityLevel_0, children [WavePlayer_3,WavePlayer_3,WavePlayer_4,WavePlayer_4]}
 {id=WavePlayer_3, properties {soundWaveAssetPtr=&/Cube_Positive_RunePan_01}}
 {id=WavePlayer_4, properties {soundWaveAssetPtr=&/Cube_Positive_RunePan_01_opt}}
]
properties {
 attenuationSettings=&/Cube_AmbientZaps_Att
 bOverrideConcurrency=true
 maxDistance=10000
 soundClassObject=/Game/Sounds/FortSoundClassesAndModes/Fort_Events_Scripted
 volumeMultiplier=0.9825
 concurrencyOverrides {maxCount=8, resolutionRule=StopOldest, volumeScale=0.75}
}
