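# Sound cue for the ambient electrical zap of the Cube in Athena events, played through the scripted events sound class.
# A Random node picks one of three equally weighted zap waves, and a Modulator varies its pitch and volume on each play.
# A dedicated attenuation asset shapes falloff, and a concurrency override caps simultaneous instances, stopping the oldest and scaling volume as they stack.
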SoundCue /Game/Athena/Sounds/Events/Cube/Cube_AmbZap_Cue
root=Modulator_4
nodes [
 {id=Modulator_4, children [Random_0], properties {pitchMax=1.2, pitchMin=0.8, volumeMin=0.5}}
 {id=Random_0, children [WavePlayer_0,WavePlayer_1,WavePlayer_2], properties {weights [1,1,1]}}
 {id=WavePlayer_0, properties {soundWaveAssetPtr=&/Cube_AmbZap_01}}
 {id=WavePlayer_1, properties {soundWaveAssetPtr=&/Cube_AmbZap_02}}
 {id=WavePlayer_2, properties {soundWaveAssetPtr=&/Cube_AmbZap_03}}
]
properties {
 attenuationSettings=&/Cube_AmbientZaps_Att
 bOverrideConcurrency=true
 maxDistance=10000
 soundClassObject=/Game/Sounds/FortSoundClassesAndModes/Fort_Events_Scripted
 volumeMultiplier=0.393
 concurrencyOverrides {maxCount=8, resolutionRule=StopOldest, volumeScale=0.75}
}
