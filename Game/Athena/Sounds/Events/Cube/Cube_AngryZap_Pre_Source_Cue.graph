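# Sound cue for the Athena Cube event's angry zap, named as the pre-zap sound at the source.
# A single retaliation wave player feeds a Modulator node left at default settings.
# It uses the shared Cube impacts attenuation and an important in-game mission sound class.
# A concurrency override allows only a couple of simultaneous instances, stopping the oldest when exceeded.
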SoundCue /Game/Athena/Sounds/Events/Cube/Cube_AngryZap_Pre_Source_Cue
root=Modulator_4
nodes [
 {id=Modulator_4, children [WavePlayer_4]}
 {id=WavePlayer_4, properties {soundWaveAssetPtr=&/Cube_AngryZapRetaliation_01}}
]
properties {
 attenuationSettings=&/Cube_Impacts_Att
 bOverrideConcurrency=true
 maxDistance=10000
 soundClassObject=&Fort_Mission_Ingame_Important_PSM
 concurrencyOverrides {maxCount=2, resolutionRule=StopOldest}
}
