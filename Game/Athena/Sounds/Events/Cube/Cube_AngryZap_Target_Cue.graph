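# Sound cue for the Cube's angry zap target, playing a zap impact sound with Cube impact attenuation.
# A quality level node sends its first two levels to a random pick of two standard impact waves and its last two to a random pick of their "_opt" variants.
# A modulator wraps the result.
# A concurrency override caps simultaneous plays at a small count, stopping the oldest.
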SoundCue /Game/Athena/Sounds/Events/Cube/Cube_AngryZap_Target_Cue
root=Modulator_4
nodes [
 {id=Modulator_4, children [QualityLevel_1]}
 {id=QualityLevel_1, children [Random_0,Random_0,Random_2,Random_2]}
 {id=Random_0, children [WavePlayer_0,WavePlayer_1], properties {weights [1,1]}}
 {id=Random_2, children [WavePlayer_2,WavePlayer_4], properties {weights [1,1]}}
 {id=WavePlayer_0, properties {soundWaveAssetPtr=&/Cube_ZapImpact_02}}
 {id=WavePlayer_1, properties {soundWaveAssetPtr=&/Cube_ZapImpact_01}}
 {id=WavePlayer_2, properties {soundWaveAssetPtr=&/Cube_ZapImpact_01_opt}}
 {id=WavePlayer_4, properties {soundWaveAssetPtr=&/Cube_ZapImpact_02_opt}}
]
properties {
 attenuationSettings=&/Cube_Impacts_Att
 bOverrideConcurrency=true
 maxDistance=10000
 soundClassObject=&Fort_Mission_Ingame_Important_PSM
 volumeMultiplier=0.5
 concurrencyOverrides {maxCount=2, resolutionRule=StopOldest}
}
