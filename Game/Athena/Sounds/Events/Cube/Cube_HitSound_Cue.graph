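# Sound cue for the Cube hit sound among the Athena Cube event sounds, currently silent because it has no root node or sound nodes.
# Only playback settings remain: Cube impact attenuation, an in-game important mission sound class and a reduced volume multiplier.
# A concurrency override limits it to a single playing instance.
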SoundCue /Game/Athena/Sounds/Events/Cube/Cube_HitSound_Cue
root=null
nodes []
properties {
 attenuationSettings=&/Cube_Impacts_Att
 bOverrideConcurrency=true
 maxDistance=10240
 soundClassObject=&Fort_Mission_Ingame_Important_PSM
 volumeMultiplier=0.5
 concurrencyOverrides {maxCount=1}
}
