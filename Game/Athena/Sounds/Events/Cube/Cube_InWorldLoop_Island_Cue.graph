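# Sound cue for the looping in-world ambience of the Cube island event, limited to one instance.
# A root mixer layers close and distant looping rune and Cube ambiences, each with its own attenuation.
# A fifth layer repeats randomly delayed zap vocals and distant thunder, with pitch and volume modulated by the DrainAmount parameter.
# Quality-level nodes swap in optimized rune loops and a smaller thunder pool.
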SoundCue /Game/Athena/Sounds/Events/Cube/Cube_InWorldLoop_Island_Cue
root=Mixer_4
nodes [
 {
  id=Attenuation_2
  children [ModulatorContinuous_1]
  properties {attenuationSettings=/Game/Athena/Sounds/Events/Attenuations/Events_Large_CloseDistant_Att}
 }
 {
  id=Attenuation_3
  children [WavePlayer_20]
  properties {
   attenuationSettings=/Game/Athena/Sounds/Events/Attenuations/Events_Large_Distant_Att
   bOverrideAttenuation=true
   attenuationOverrides {falloffDistance=6000, stereoSpread=1024}
  }
 }
 {
  id=Attenuation_4
  children [WavePlayer_19]
  properties {
   attenuationSettings=/Game/Athena/Sounds/Events/Attenuations/Events_Large_Close_Att
   bOverrideAttenuation=true
   attenuationOverrides {falloffDistance=5000, stereoSpread=512, attenuationShapeExtents {x=1000}}
  }
 }
 {
  id=Attenuation_8
  children [QualityLevel_8]
  properties {attenuationSettings=&Events_Large_Distant_Att}
 }
 {id=Attenuation_9, children [QualityLevel_9], properties {attenuationSettings=&Events_Large_Close_Att}}
 {id=Delay_1, children [Mixer_8], properties {delayMax=10, delayMin=0.5}}
 {id=Delay_2, children [Random_5], properties {delayMax=30, delayMin=10}}
 {id=Looping_1, children [Delay_1]}
 {id=Looping_2, children [Delay_2]}
 {id=Mixer_4, children [Attenuation_8, Attenuation_9, Attenuation_2, Attenuation_3, Attenuation_4], properties {inputVolume [1,1,1,1.5,1.5]}}
 {id=Mixer_6, children [Looping_2,Looping_1], properties {inputVolume [1.25,1]}}
 {id=Mixer_8, children [QualityLevel_11], properties {inputVolume [0.4]}}
 {
  id=ModulatorContinuous_1
  children [Mixer_6]
  properties {pitchModulationParams {maxOutput=1.2, minOutput=0.8, parameterName=DrainAmount}, volumeModulationParams {parameterName=DrainAmount}}
 }
 {id=QualityLevel_11, children [Random_0,Random_1,Random_1,Random_1]}
 {id=QualityLevel_8, children [WavePlayer_26, WavePlayer_26, WavePlayer_29, WavePlayer_29]}
 {id=QualityLevel_9, children [WavePlayer_28, WavePlayer_28, WavePlayer_30, WavePlayer_30]}
 {
  id=Random_0
  children [WavePlayer_27, WavePlayer_31, WavePlayer_32, WavePlayer_33, WavePlayer_34, WavePlayer_35, WavePlayer_36, WavePlayer_37, WavePlayer_38]
  properties {weights [1,1,1,1,1,1,1,1,1]}
 }
 {id=Random_1, children [WavePlayer_39, WavePlayer_40, WavePlayer_41], properties {weights [1,1,1]}}
 {id=Random_5, children [WavePlayer_21, WavePlayer_22], properties {weights [1,1]}}
 {id=WavePlayer_19, properties {bLooping=true, soundWaveAssetPtr=&/Cube_AmbientClose_01}}
 {id=WavePlayer_20, properties {bLooping=true, soundWaveAssetPtr=&/Cube_Ambient_01}}
 {id=WavePlayer_21, properties {soundWaveAssetPtr=/Game/Athena/Sounds/Events/CubeIsland/Cube_SpawnerZapVocal_02}}
 {id=WavePlayer_22, properties {soundWaveAssetPtr=/Game/Athena/Sounds/Events/CubeIsland/Cube_SpawnerZapVocal_01}}
 {id=WavePlayer_26, properties {bLooping=true, soundWaveAssetPtr=/Game/Athena/Sounds/Events/CubeIsland/DudeBro_Amb_Rne_Dist_01}}
 {id=WavePlayer_27, properties {soundWaveAssetPtr=/Game/Sounds/Fort_GamePlay_Sounds/Storms/Dist_Thunder_001}}
 {id=WavePlayer_28, properties {bLooping=true, soundWaveAssetPtr=/Game/Athena/Sounds/Events/CubeIsland/DudeBro_Amb_Rne_Close_01}}
 {id=WavePlayer_29, properties {bLooping=true, soundWaveAssetPtr=/Game/Athena/Sounds/Events/CubeIsland/DudeBro_Amb_Rne_Dist_01_opt}}
 {id=WavePlayer_30, properties {bLooping=true, soundWaveAssetPtr=/Game/Athena/Sounds/Events/CubeIsland/DudeBro_Amb_Rne_Close_01_opt}}
 {id=WavePlayer_31, properties {soundWaveAssetPtr=/Game/Sounds/Fort_GamePlay_Sounds/Storms/Dist_Thunder_002}}
 {id=WavePlayer_32, properties {soundWaveAssetPtr=/Game/Sounds/Fort_GamePlay_Sounds/Storms/Dist_Thunder_003}}
 {id=WavePlayer_33, properties {soundWaveAssetPtr=/Game/Sounds/Fort_GamePlay_Sounds/Storms/Dist_Thunder_004}}
 {id=WavePlayer_34, properties {soundWaveAssetPtr=/Game/Sounds/Fort_GamePlay_Sounds/Storms/Dist_Thunder_005}}
 {id=WavePlayer_35, properties {soundWaveAssetPtr=/Game/Sounds/Fort_GamePlay_Sounds/Storms/Dist_Thunder_006}}
 {id=WavePlayer_36, properties {soundWaveAssetPtr=/Game/Sounds/Fort_GamePlay_Sounds/Storms/Dist_Thunder_007}}
 {id=WavePlayer_37, properties {soundWaveAssetPtr=/Game/Sounds/Fort_GamePlay_Sounds/Storms/Dist_Thunder_008}}
 {id=WavePlayer_38, properties {soundWaveAssetPtr=/Game/Sounds/Fort_GamePlay_Sounds/Storms/Dist_Thunder_009}}
 {id=WavePlayer_39, properties {soundWaveAssetPtr=&Dist_Thunder_001}}
 {id=WavePlayer_40, properties {soundWaveAssetPtr=&Dist_Thunder_002}}
 {id=WavePlayer_41, properties {soundWaveAssetPtr=&Dist_Thunder_003}}
]
properties {
 attenuationSettings=&Events_Large_CloseDistant_Att
 bOverrideConcurrency=true
 maxDistance=30000
 soundClassObject=/Game/Sounds/FortSoundClassesAndModes/Fort_Events_Scripted
 volumeMultiplier=1.0025
 concurrencyOverrides {maxCount=1}
}
