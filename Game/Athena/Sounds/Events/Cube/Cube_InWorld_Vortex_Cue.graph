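# Sound cue for the in-world Cube vortex among the Athena event sounds.
# Its node graph is empty with no root, so the file references no sound wave.
# It carries only playback settings: large close-to-distant event attenuation, the scripted-events sound class and a concurrency override allowing a single instance.
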SoundCue /Game/Athena/Sounds/Events/Cube/Cube_InWorld_Vortex_Cue
root=null
nodes []
properties {
 attenuationSettings=/Game/Athena/Sounds/Events/Attenuations/Events_Large_CloseDistant_Att
 bOverrideConcurrency=true
 maxDistance=30000
 soundClassObject=/Game/Sounds/FortSoundClassesAndModes/Fort_Events_Scripted
 volumeMultiplier=1.0025
 concurrencyOverrides {maxCount=1}
}
