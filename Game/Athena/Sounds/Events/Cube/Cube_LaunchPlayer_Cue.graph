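# Sound cue for the Cube launching a player in Athena, playing a zap impact routed to the scripted events sound class with launch attenuation.
# A root Modulator feeds a quality-level switch: the first two slots randomly pick one of two standard waves.
# The last two slots add another Modulator over a random pick of two "_opt" variants.
# A concurrency override allows one instance per owner, stopping the oldest.
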SoundCue /Game/Athena/Sounds/Events/Cube/Cube_LaunchPlayer_Cue
root=Modulator_1
nodes [
 {id=Modulator_1, children [QualityLevel_0]}
 {id=Modulator_2, children [Random_3]}
 {id=QualityLevel_0, children [Random_2,Random_2,Modulator_2,Modulator_2]}
 {id=Random_2, children [WavePlayer_6,WavePlayer_7], properties {weights [1,1]}}
 {id=Random_3, children [WavePlayer_8,WavePlayer_9], properties {weights [1,1]}}
 {id=WavePlayer_6, properties {soundWaveAssetPtr=&/Cube_ZapImpact_01}}
 {id=WavePlayer_7, properties {soundWaveAssetPtr=&/Cube_ZapImpact_02}}
 {id=WavePlayer_8, properties {soundWaveAssetPtr=&/Cube_ZapImpact_01_opt}}
 {id=WavePlayer_9, properties {soundWaveAssetPtr=&/Cube_ZapImpact_02_opt}}
]
properties {
 attenuationSettings=&/Cube_Launch_Att
 bOverrideConcurrency=true
 maxDistance=20000
 soundClassObject=/Game/Sounds/FortSoundClassesAndModes/Fort_Events_Scripted
 volumeMultiplier=0.786
 concurrencyOverrides {bLimitToOwner=true, maxCount=1, resolutionRule=StopOldest}
}
