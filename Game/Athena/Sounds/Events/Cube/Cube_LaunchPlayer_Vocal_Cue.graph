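# Sound cue named for the vocal played when the Cube launches a player, currently empty with no root node or sound nodes, so it has no audio source.
# It still carries playback settings: Cube impacts attenuation, the in-game mission sound class and a lowered pitch multiplier.
# A concurrency override limits it to a single instance and rejects new plays while one is active.
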SoundCue /Game/Athena/Sounds/Events/Cube/Cube_LaunchPlayer_Vocal_Cue
root=null
nodes []
properties {
 attenuationSettings=&/Cube_Impacts_Att
 bOverrideConcurrency=true
 maxDistance=10000
 pitchMultiplier=0.75
 soundClassObject=/Game/Sounds/FortSoundClassesAndModes/Fort_Mission_IngameSounds
 concurrencyOverrides {maxCount=1, resolutionRule=PreventNew}
}
