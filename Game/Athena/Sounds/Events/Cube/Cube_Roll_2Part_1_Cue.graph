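# Sound cue for the Athena Cube event's roll sound, playing one of two roll waves chosen at random with equal weight.
# The random pick feeds an attenuation node using Cube_Roll_Att, then a root mixer whose second input is empty.
# A concurrency override caps it at one instance, and it uses the important in-game mission sound class with a large max distance.
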SoundCue /Game/Athena/Sounds/Events/Cube/Cube_Roll_2Part_1_Cue
root=Mixer_0
nodes [
 {id=Attenuation_0, children [Random_0], properties {attenuationSettings=&/Cube_Roll_Att}}
 {id=Mixer_0, children [Attenuation_0, null], properties {inputVolume [1,1.5]}}
 {id=Random_0, children [WavePlayer_0,WavePlayer_1], properties {weights [1,1]}}
 {id=WavePlayer_0, properties {soundWaveAssetPtr=&/Cube_Roll_01}}
 {id=WavePlayer_1, properties {soundWaveAssetPtr=&/Cube_Roll_02}}
]
properties {
 bOverrideConcurrency=true
 maxDistance=30000
 soundClassObject=&Fort_Mission_Ingame_Important_PSM
 volumeMultiplier=0.8
 concurrencyOverrides {maxCount=1}
}
